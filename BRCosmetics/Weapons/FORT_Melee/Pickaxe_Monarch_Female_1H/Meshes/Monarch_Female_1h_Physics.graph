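# Physics asset for the one-handed female Monarch pickaxe, setting up a short dangling rope for physics-driven secondary motion.
# A capsule on the handle bone anchors a two-link chain: a capsule on dyn_rope_01, then a flattened box on dyn_rope_02.
# Two constraints join the links with swing and twist limit angles, with one swing axis left free at the handle joint.
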PhysicsAsset /BRCosmetics/Weapons/FORT_Melee/Pickaxe_Monarch_Female_1H/Meshes/Monarch_Female_1h_Physics
bodies {
 Melee_TwoHanded_Handle [{shapeType=Capsule, length=34, radius=1.5}]
 dyn_rope_01 [{shapeType=Capsule, length=5.5, radius=1, center=<3.75 0 0>, rotation=<rot 90 0 0>}]
 dyn_rope_02 [{shapeType=Box, extentX=7, extentY=6, extentZ=1.5, shapeName=dyn_rope_02_box, center=<3.75 0 0>}]
}
constraints [
 {bone1Name=dyn_rope_02, bone2Name=dyn_rope_01, swing1LimitDegrees=30, swing2LimitDegrees=75, twistLimitDegrees=45}
 {
  bone1Name=dyn_rope_01
  bone2Name=Melee_TwoHanded_Handle
  swing1LimitDegrees=75
  swing2LimitDegrees=45
  swing2Motion=Free
  twistLimitDegrees=45
 }
]
